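; Gameplay ability for a Galileo Camp driving quest that grants an urgent transient quest and tracks arrival at a destination.
; On event activation it continues only if the owning pawn has a starting-location tag, otherwise it ends.
; It then waits for the destination tag to be added, sends a quest stat event with the source tags, and ends.
; A vehicle-passenger helper is defined but not called here.
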(event Ability|EventActivateAbilityFromEvent
  (params (EventData :type Gameplay%20Event%20Data%20Structure%20%28by%20ref%29))
  ;[0 0 0] Activate ability - check if player is in correct starting location
  (|CallParentFunction EventData)
  (if (GameplayTags|HasAnyMatchingGameplayTags (Variables|GAVars|Data|GetOwningPlayerPawn) (Variables|Quest|GetStartingLocationTags))
    (CallFunction|GrantUrgentQuesttoOwningPlayer)
    (CallFunction|HandleEnteringDestination)
    (else
      (Ability|EndAbility))))

(event Custom|HandleEnteringDestination
  ;[0 0 0] Listen for destination tag to be added
  (Ability|Tasks|WaitGameplayTagAdd (Variables|Quest|GetDestinationTag) 0 true
    (:Added
      (Quest|SendComplexCustomStatEvent (Variables|GAVars|Data|GetOwningPlayerQuestManager) 0 (Variables|Quest|GetQuestSourceTags) "(GameplayTags=)" 1)
      (Ability|EndAbility))))

(fn GetPassengersinOwningPlayerVehicle
  ;[0 0 0] Helper function for getting all the passengers currently in the owning player's vehicle.
  (return (Vehicle|GetAllPassengers (Vehicle|GetVehicleInterface (Variables|GAVars|Data|GetOwningPlayerPawn)))))

(fn GrantUrgentQuestToOwningPlayer
  ;[0 0 0] Helper function for granting the specified urgent quest to the ability owner
  (bind _asfort_player_controller_athena (Utilities|Casting|CastToFortPlayerControllerAthena (Fortnite|GetFortPlayerControllerfromActor (Variables|GAVars|Data|GetOwningPlayerPawn)))
    (:then
      (Class|FortControllerComponentTransientQuests|GrantTransientQuest (Class|FortPlayerControllerAthena|GetTransientQuestsComponent _asfort_player_controller_athena) (Variables|Quest|GetUrgentQuestToGrant) 0)
      (return))
    (:CastFailed
      (return))))
