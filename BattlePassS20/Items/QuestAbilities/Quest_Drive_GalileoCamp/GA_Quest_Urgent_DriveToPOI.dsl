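; GA_Quest_Urgent_DriveToPOI is a gameplay ability for the urgent drive-to-POI quest, located in the Galileo camp quest folder.
; When activated from a gameplay event, it runs the parent logic first.
; On the server it then calls FailQuest and ends the ability, so exiting the vehicle fails the quest.
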(event Ability|EventActivateAbilityFromEvent
  (params (EventData :type Gameplay%20Event%20Data%20Structure%20%28by%20ref%29))
  ;[1 1 1] When the player exits the vehicle, fail the quest
  (|CallParentFunction EventData)
  (if (Networking|IsServer)
    (CallFunction|FailQuest)
    (Ability|EndAbility)))
